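# Skeletal mesh character material for the Super Levels skin of RoseDust, using material attributes with Default Lit and Subsurface Profile shading.
# Base PreFX feeds chained S22 tier functions T1, T2 and T3, then character and cinematic PostFX.
# The tiers share an SL_FX mask parameter, with functions and default texture borrowed from the Bites character.
# A RoseDust thin-film, inkline and highlight branch also reads PreFX, but its result feeds no other node.
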Material /BRCosmetics/Characters/Player/Female/Medium/Bodies/F_MED_RoseDust/Skins/Super_Levels/Materials/M_F_RoseDust_SL
graph {
 properties {
  bAutomaticallySetUsageInEditor=false
  bUseMaterialAttributes=true
  bUsedWithMeshDeformer=true
  bUsedWithMorphTargets=true
  bUsedWithSkeletalMesh=true
  pixelDepthOffsetMode=PDOM_Legacy
  shadingModel=MSM_FromMaterialExpression
  subsurfaceProfile=&SS_HeroSkin_02
  usedShadingModels=Default Lit | Subsurface Profile
 }
 nodes [
  n1 = MaterialFunctionCall() {materialFunction=&BaseFN_CharacterFunction_PreFX}
  n2 = TextureSampleParameter2D() {
   group=S22_SL
   parameterName=SL_FX
   samplerType=SAMPLERTYPE_Masks
   sortPriority=1
   texture=/BRCosmetics/Characters/Player/Male/Medium/Bodies/M_MED_Bites/Skins/Super_Levels/Textures/T_Character_Bites_Body_SL_FX
  }
  n3 = MaterialFunctionCall!(MA=@n1.Result, SL_FX=@n2.RGBA, AO=@n1.Result) {materialFunction=/BRCosmetics/Characters/Player/Male/Medium/Bodies/M_MED_Bites/Skins/Super_Levels/Materials/MF_SL_S22_T1}
  n4 = MaterialFunctionCall(
   AO=@n1.AO
   Distance Blend=@n1.AO
   Fresnel=@n1.AO
   Gmap=@n1.AO
   Pre Skinned Local Position=@n1.AO
   Material=MaterialFunctionCall!(
    MA=MaterialFunctionCall!(MA=@n3.Result, SL_FX=@n2.RGBA, AO=@n1.AO, PreSkinnedLocalPosition_VS, PreSkinnedLocalNormal_PS) {materialFunction=/BRCosmetics/Characters/Player/Male/Medium/Bodies/M_MED_Bites/Skins/Super_Levels/Materials/MF_SL_S22_T2}.Result
    SL_FX=@n2.RGBA
    AO=@n1.AO
   ) {materialFunction=/BRCosmetics/Characters/Player/Male/Medium/Bodies/M_MED_Bites/Skins/Super_Levels/Materials/MF_SL_S22_T3}.Result
  ) {materialFunction=&BaseFN_CharacterFunction_PostFX}
  MaterialFunctionCall(In=@n4.Result) {materialFunction=&CINE_MF_CharPostFX}
  n5 = NamedRerouteUsage()
  MaterialFunctionCall!(
   MA=MaterialFunctionCall!(DFMap=MaterialFunctionCall!(Material Attributes=@n1.Result, SkinFX=@n5) {materialFunction=/Game/Characters/Player/Female/Medium/Bodies/F_MED_RoseDust/Materials/MF_RoseDust_ThinFilm}.Result) {materialFunction=/Game/Characters/Player/Female/Medium/Bodies/F_MED_RoseDust/Materials/MF_RoseDust_DF_Inklines}.Result
   FX=NamedRerouteUsage()
  ) {materialFunction=/Game/Characters/Player/Female/Medium/Bodies/F_MED_RoseDust/Materials/MF_RoseDust_Highlights}
  n6 = TextureSampleParameter2D() {
   group=Skin FX
   parameterName=SkinFX_Mask
   samplerType=SAMPLERTYPE_Masks
   sortPriority=1
   texture=&T_BlackMask
  }
  NamedRerouteDeclaration(None=@n6.RGBA) {name=SkinFX, nodeColor=<rgba 0 1 0.8359 1>}
 ]
}
